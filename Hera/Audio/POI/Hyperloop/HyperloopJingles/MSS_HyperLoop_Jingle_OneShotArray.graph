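# Stereo MetaSound source for the HyperLoop POI jingle, playing one of three shuffled music stingers per play trigger.
# On Play advances a shared-state shuffle into a wave player, with both channels scaled by a float variable.
# Output goes to the HyperLoop jingle audio bus, with base submix and submix sends disabled.
# A random-delay chain writes a trigger variable but has no incoming trigger or reader, so nothing retriggers playback.
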MetaSoundSource /Hera/Audio/POI/Hyperloop/HyperloopJingles/MSS_HyperLoop_Jingle_OneShotArray
inputs [
 Sounds {type=WaveAsset:Array}
 UE.Source.OnPlay {type=Trigger}
 UE.Source.StartTime.StartTime {type=Time}
]
interfaces [UE.OutputFormat.Stereo:1.0, UE.Source.StartTime:1.0, UE.Source:1.1]
outputs [
 "UE.OutputFormat.Stereo.Audio:0" {type=Audio}
 "UE.OutputFormat.Stereo.Audio:1" {type=Audio}
]
pages [Default {
  input_defaults {
   UE.Source.OnPlay=null
   UE.Source.StartTime.StartTime=0
   Sounds [
    &/MX_S41_HyperLoop_Jingle_STNG_1
    &/MX_S41_HyperLoop_Jingle_STNG_2
    &/MX_S41_HyperLoop_Jingle_STNG_3
   ]
  }
  nodes [
   n1 = VariableAccessor.Float(Variable=InitVariable.Float().Variable)
   TriggerAnyTriggerAny2_0 = "TriggerAny.Trigger Any (2)"(In 0=^UE.Source.OnPlay)
   n2 = Array.Shuffle:WaveAsset:Array(In Array=^Sounds, Next=@TriggerAnyTriggerAny2_0.Out) {also {defaults {Enable Shared State="<MetaSoundBuilderResult.SUCCEEDED: 0>"}}}
   WavePlayer_0 = UE.Wave Player:Stereo(Play=@n2.On Next, Wave Asset=@n2.Value) {also {version [1,1]}}
   Multiply_0 = UE.Multiply:Audio by Float(AdditionalOperands=@n1.Value, PrimaryOperand=@WavePlayer_0.Out Right)
   Multiply_1 = UE.Multiply:Audio by Float(AdditionalOperands=@n1.Value, PrimaryOperand=@WavePlayer_0.Out Left)
   n3 = UE.RandomTime() {also {defaults {Max=60, Min=20}}}
   TriggerDelay_0 = UE.Trigger Delay(In=@n3.On Next, Delay Time=@n3.Value)
   VariableMutator.Trigger(Value=@TriggerDelay_0.Out, Variable=VariableDeferredAccessor.Trigger(Variable=InitVariable.Trigger().Variable).Variable)
  ]
  links [{dst="@UE.OutputFormat.Stereo.Audio:1", src=Multiply_0.Out}, {dst="@UE.OutputFormat.Stereo.Audio:0", src=Multiply_1.Out}]
 }]
properties {
 assetImportData=&self:AssetImportData
 bEnableBaseSubmix=false
 bEnableSubmixSends=false
 bLooping=true
 duration=10000
 numChannels=2
 outputFormat=Stereo
 sampleRate=48000
 busSends [{audioBus=&/AB_HyperLoopJingle, sendLevel=1}]
}
